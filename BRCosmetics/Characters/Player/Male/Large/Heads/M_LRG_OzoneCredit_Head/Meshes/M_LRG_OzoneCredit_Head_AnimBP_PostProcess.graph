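# Post-process AnimBlueprint for the M_LRG_OzoneCredit_Head cosmetic character head.
# The AnimGraph applies a fixed additive bone-space yaw offset to the head bone of the input pose.
# Each update computes an inverted, per-axis scaled head-to-neck_02 rotation delta when no emote plays or the YouBoreMe montage is active.
# That result is stored in `head rot`, which the AnimGraph does not read.
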AnimBlueprint /BRCosmetics/Characters/Player/Male/Large/Heads/M_LRG_OzoneCredit_Head/Meshes/M_LRG_OzoneCredit_Head_AnimBP_PostProcess
target_skeleton=/Game/Characters/Player/Common/Fortnite_Base_Head/Fortnite_Base_Head_Export_Skeleton
graphs [
 AnimGraph {nodes [
   n1 = "Transform (Modify) Bone - Bone: head"(ComponentPose=Local To Component(LocalPose=Input Pose()), Alpha=1, Translation="0,0,0", Rotation="0,-50,0", Scale="1,1,1") {
    rotationMode=BMM_Additive
    rotationSpace=BCS_BoneSpace
    translationMode=BMM_Additive
    alphaBoolBlend {blendInTime=0.5, blendOption=Linear, blendOutTime=0.5}
    alphaScaleBiasClamp {bInterpResult=true, interpSpeedDecreasing=5, interpSpeedIncreasing=5}
    boneToModify {boneName=head}
    rotation {yaw=-50}
   }
   Output Pose(Result=Component To Local(ComponentPose=@n1))
  ]}
 EventGraph {nodes [
   Event BlueprintUpdateAnimation()
   > n1 = Sequence!(execute) -> (then_0, then_1, then_2:Exec)
   > Set BodyMesh(BodyMesh=GetOwnerSkeletalMeshComponent())
   n2 = Branch(execute=@n1.then_1, Condition=NOT Boolean(A=Get bIsPlayingEmote()))
   n3 = GetOwnerSkeletalMeshComponent()
   n4 = BreakRotator(InRot=BreakTransform(InTransform=GetDeltaTransformFromRefPose(self=@n3, BoneName=head, BaseName=neck_02)).Rotation)
   n5 = MakeRotator(
    Roll=float * float!(A=@n4, B=-1, C:double=Get X percentage() -> (X percentage:float)) -> (ReturnValue)
    Pitch=float * float(A=@n4.Pitch, B=-0.8, C=Get Y percentage() -> (Y percentage:float))
    Yaw=float * float(A=@n4.Yaw, B=-1, C=Get Z percentage() -> (Z percentage:float))
   )
   > Set head rot(head rot:Rotator=@n5) -> (Output_Get:Rotator)
   Branch(execute=@n2.else, Condition=Get IsYouBoreMe() -> (IsYouBoreMe:bool))
   > Set head rot(head rot=@n5)
   n6 = GetAnimInstance(self=GetOwnerSkeletalMeshComponent())
   Is Valid(exec=@n1.then_2, InputObject=@n6)
   > Set IsYouBoreMe(IsYouBoreMe:bool=Montage_IsPlaying(self=@n6, Montage=/BRCosmetics/Animation/Game/MainPlayer/Emotes/YouBoreMe/Emote_You_Bore_Me_CLM_M)) -> (Output_Get:bool)
  ]}
]
variables [
 BodyMesh {type=SkeletalMeshComponent}
 head rot {type=Rotator}
 IsCelebrationEmote {type=boolean}
 IsCelebrationEmoteFloat {type=number}
 IsYouBoreMe {type=boolean}
 X percentage {type=number, default=0.8}
 Y percentage {type=number, default=1}
 Z percentage {type=number, default=0.8}
]
